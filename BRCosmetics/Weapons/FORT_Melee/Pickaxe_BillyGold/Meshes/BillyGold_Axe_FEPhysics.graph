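# Physics asset for the BillyGold pickaxe, defining capsule bodies for a dynamic bone chain on the weapon and for thigh and calf bones on both legs.
# Four dyn_BillyGold bodies link in sequence from base_BillyGold, with asymmetric swing limits and zero-degree twist limits.
# Each calf joins its thigh with free swing and twist, and no constraint connects the legs to the weapon chain.
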PhysicsAsset /BRCosmetics/Weapons/FORT_Melee/Pickaxe_BillyGold/Meshes/BillyGold_Axe_FEPhysics
bodies {
 base_BillyGold [{shapeType=Capsule, length=5, radius=10, shapeName=base_BillyGold_capsule, center=<1 0 0>, rotation=<rot 90 0 0>}]
 calf_l [{shapeType=Capsule, length=30, radius=8, shapeName=thigh_l_capsule, center=<-26 0 -1.8016>, rotation=<rot 90 180 0>}]
 calf_r [{shapeType=Capsule, length=30, radius=9.5, shapeName=thigh_r_capsule, center=<26.2104 0 -2.3098>, rotation=<rot 84.9612 0 0.0002>}]
 dyn_BillyGold_01 [{shapeType=Capsule, length=4, radius=8, shapeName=dyn_BillyGold_01_capsule, center=<1.5 0 0>, rotation=<rot 90 0 0>}]
 dyn_BillyGold_02 [{shapeType=Capsule, length=4, radius=8, shapeName=dyn_BillyGold_02_capsule, center=<0.2 0 0>, rotation=<rot 90 0 0>}]
 dyn_BillyGold_03 [{shapeType=Capsule, length=4, radius=9, shapeName=dyn_BillyGold_03_capsule, center=<6 0 0>, rotation=<rot 90 0 0>}]
 dyn_BillyGold_04 [{shapeType=Capsule, length=4, radius=9, shapeName=dyn_BillyGold_04_capsule, center=<8.649 0 0>, rotation=<rot 90 0 0>}]
 thigh_l [{shapeType=Capsule, length=34.6616, radius=9.5479, shapeName=thigh_l_capsule, center=<-22.1907 -0.6066 -0.996>, rotation=<rot 90 180 0>}]
 thigh_r [{shapeType=Capsule, length=34.6616, radius=9.5479, shapeName=thigh_r_capsule, center=<22.1908 0.6066 -0.043>, rotation=<rot 90 0 0>}]
}
constraints [
 {bone1Name=dyn_BillyGold_01, bone2Name=base_BillyGold, swing1LimitDegrees=50, swing2LimitDegrees=5, twistLimitDegrees=0}
 {bone1Name=dyn_BillyGold_02, bone2Name=dyn_BillyGold_01, swing1LimitDegrees=25, swing2LimitDegrees=10, twistLimitDegrees=0}
 {bone1Name=dyn_BillyGold_03, bone2Name=dyn_BillyGold_02, swing1LimitDegrees=20, swing2LimitDegrees=10, twistLimitDegrees=0}
 {bone1Name=dyn_BillyGold_04, bone2Name=dyn_BillyGold_03, swing1LimitDegrees=30, swing2LimitDegrees=25, twistLimitDegrees=0}
 {
  bone1Name=calf_l
  bone2Name=thigh_l
  swing1LimitDegrees=45
  swing1Motion=Free
  swing2LimitDegrees=45
  swing2Motion=Free
  twistLimitDegrees=45
  twistMotion=Free
 }
 {
  bone1Name=calf_r
  bone2Name=thigh_r
  swing1LimitDegrees=45
  swing1Motion=Free
  swing2LimitDegrees=45
  swing2Motion=Free
  twistLimitDegrees=45
  twistMotion=Free
 }
]
