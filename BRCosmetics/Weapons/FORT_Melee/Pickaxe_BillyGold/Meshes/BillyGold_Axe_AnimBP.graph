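# Animation Blueprint for the BillyGold pickaxe, adding physics-driven secondary motion and pose tweaks to the axe.
# World-space bone modifiers on pelvis, leg and root bones build a cached pose; the event graph samples matching body-mesh sockets.
# That pose feeds two chained RigidBody simulations (regular and front-end physics assets), then additive dyn_BillyGold_05 and handle offsets.
# Update logic selects sim-space settings and alphas by front-end state, and tracks gravity, movement, crouch and swing values.
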AnimBlueprint /BRCosmetics/Weapons/FORT_Melee/Pickaxe_BillyGold/Meshes/BillyGold_Axe_AnimBP
target_skeleton=&/BillyGold_Axe_Skeleton
graphs [
 AnimGraph {nodes [
   n1 = RigidBody!(
    ComponentPose=Local To Component(LocalPose=Use cached pose 'CR_Cache_Input'())
    Alpha
    ExternalForce
    SimSpaceSettings:Sim Space Settings Structure=(WorldAlpha=1,VelocityScaleZ=0.5,DampingAlpha=0,MaxLinearVelocity=700,MaxAngularVelocity=8,MaxLinearAcceleration=3500,MaxAngularAcceleration=45,ExternalLinearDragV=(X=0.5,Y=0.5,Z=0.35),ExternalLinearVelocity=(X=0,Y=0,Z=0),ExternalAngularVelocity=(X=0,Y=0,Z=0))
   ) -> (Pose) {
    bClampLinearTranslationLimitToRefPose=true
    bOverrideWorldGravity=true
    overridePhysicsAsset=&/BillyGold_Axe_Physics
    simulationSpace=BaseBoneSpace
    alphaBoolBlend {blendOption=Linear}
    baseBoneRef {boneName=b_AR_Root}
    overrideWorldGravity {z=-980}
    simSpaceSettings {
     dampingAlpha=0
     maxAngularAcceleration=45
     maxAngularVelocity=8
     maxLinearAcceleration=3500
     maxLinearVelocity=700
     velocityScaleZ=0.5
     worldAlpha=1
     externalLinearDragV=<0.5 0.5 0.35>
    }
   }
   n2 = RigidBody(
    ComponentPose=@n1
    SimSpaceSettings=(WorldAlpha=1,VelocityScaleZ=0.5,DampingAlpha=0,MaxLinearVelocity=700,MaxAngularVelocity=8,MaxLinearAcceleration=3500,MaxAngularAcceleration=45,ExternalLinearDragV=(X=0.5,Y=0.5,Z=0.35),ExternalLinearVelocity=(X=0,Y=0,Z=0),ExternalAngularVelocity=(X=0,Y=0,Z=0))
   ) {
    bClampLinearTranslationLimitToRefPose=true
    bOverrideWorldGravity=true
    overridePhysicsAsset=&/BillyGold_Axe_FEPhysics
    simulationSpace=BaseBoneSpace
    alphaBoolBlend {blendOption=Linear}
    baseBoneRef {boneName=b_AR_Root}
    overrideWorldGravity {z=-1200}
    simSpaceSettings {
     dampingAlpha=0
     maxAngularAcceleration=45
     maxAngularVelocity=8
     maxLinearAcceleration=3500
     maxLinearVelocity=700
     velocityScaleZ=0.5
     worldAlpha=1
     externalLinearDragV=<0.5 0.5 0.35>
    }
   }
   n3 = "Transform (Modify) Bone - Bone: dyn_BillyGold_05"(ComponentPose=@n2, Alpha=1, Translation="2,0,7", Rotation="80,0,0", Scale="1,1,1") {
    rotationMode=BMM_Additive
    rotationSpace=BCS_BoneSpace
    translationMode=BMM_Additive
    translationSpace=BCS_BoneSpace
    alphaBoolBlend {blendOption=Linear}
    alphaScaleBiasClamp {bInterpResult=true, interpSpeedDecreasing=5, interpSpeedIncreasing=5}
    boneToModify {boneName=dyn_BillyGold_05}
    rotation {pitch=80}
    translation {x=2, z=7}
   }
   n4 = "Transform (Modify) Bone - Bone: dyn_BillyGold_05"(ComponentPose=@n3, Alpha=1, Translation="2,0,7", Rotation="30,0,0", Scale="1,1,1") {
    rotationMode=BMM_Additive
    rotationSpace=BCS_BoneSpace
    translationMode=BMM_Additive
    translationSpace=BCS_BoneSpace
    alphaBoolBlend {blendOption=Linear}
    alphaScaleBiasClamp {bInterpResult=true, interpSpeedDecreasing=5, interpSpeedIncreasing=5}
    boneToModify {boneName=dyn_BillyGold_05}
    rotation {pitch=30}
    translation {x=2, z=7}
   }
   n5 = "Transform (Modify) Bone - Bone: Melee_TwoHanded_Handle"!(ComponentPose=@n4, bAlphaBoolEnabled:bool=True, Translation="0,0,0", Rotation="0,0,-30", Scale="1,1,1") -> (Pose) {
    alphaInputType=Bool
    rotationMode=BMM_Additive
    rotationSpace=BCS_BoneSpace
    translationMode=BMM_Additive
    translationSpace=BCS_BoneSpace
    alphaBoolBlend {blendOption=Linear}
    alphaScaleBiasClamp {bInterpResult=true, interpSpeedDecreasing=5, interpSpeedIncreasing=5}
    boneToModify {boneName=Melee_TwoHanded_Handle}
    rotation {roll=-30}
   }
   n6 = "Transform (Modify) Bone - Bone: Melee_TwoHanded_Handle"!(ComponentPose=@n5, Alpha:float=1, Translation="0,0,0", Rotation="0,0,30", Scale="1,1,1") -> (Pose) {
    rotationMode=BMM_Additive
    rotationSpace=BCS_BoneSpace
    translationMode=BMM_Additive
    translationSpace=BCS_BoneSpace
    alphaBoolBlend {blendOption=Linear}
    alphaScaleBiasClamp {bInterpResult=true, interpSpeedDecreasing=5, interpSpeedIncreasing=5}
    boneToModify {boneName=Melee_TwoHanded_Handle}
    rotation {roll=30}
   }
   Output Pose(Result=Component To Local(ComponentPose=@n6)) {layerGroup=None}
   AnimGraphNode_SubInput_0 = Input Pose()
   n7 = "Transform (Modify) Bone - Bone: body_root"(ComponentPose=Local To Component(LocalPose=@AnimGraphNode_SubInput_0), Alpha=1, Translation="0,0,0", Rotation="0,0,0", Scale="1,1,1") {
    rotationMode=BMM_Replace
    rotationSpace=BCS_WorldSpace
    scaleSpace=BCS_BoneSpace
    translationMode=BMM_Replace
    translationSpace=BCS_WorldSpace
    alphaBoolBlend {blendOption=Linear}
    alphaScaleBiasClamp {bInterpResult=true, interpSpeedDecreasing=1, interpSpeedIncreasing=1, inRange {max=0, min=-40}, outRange {max=0, min=1}}
    boneToModify {boneName=body_root}
   }
   n8 = "Transform (Modify) Bone - Bone: pelvis"(ComponentPose=@n7, Alpha=1, Translation="0,0,0", Rotation="0,0,0", Scale="1,1,1") {
    rotationMode=BMM_Replace
    rotationSpace=BCS_WorldSpace
    scaleSpace=BCS_BoneSpace
    translationMode=BMM_Replace
    translationSpace=BCS_WorldSpace
    alphaBoolBlend {blendOption=Linear}
    alphaScaleBiasClamp {inRange {max=0, min=-40}, outRange {max=0, min=1}}
    boneToModify {boneName=pelvis}
   }
   n9 = "Transform (Modify) Bone - Bone: thigh_l"(ComponentPose=@n8, Alpha=1, Translation="0,0,0", Rotation="0,0,0", Scale="1,1,1") {
    rotationMode=BMM_Replace
    rotationSpace=BCS_WorldSpace
    scaleSpace=BCS_BoneSpace
    translationMode=BMM_Replace
    translationSpace=BCS_WorldSpace
    alphaBoolBlend {blendOption=Linear}
    alphaScaleBiasClamp {inRange {max=0, min=-40}, outRange {max=0, min=1}}
    boneToModify {boneName=thigh_l}
   }
   n10 = "Transform (Modify) Bone - Bone: thigh_r"(ComponentPose=@n9, Alpha=1, Translation="0,0,0", Rotation="0,0,0", Scale="1,1,1") {
    rotationMode=BMM_Replace
    rotationSpace=BCS_WorldSpace
    scaleSpace=BCS_BoneSpace
    translationMode=BMM_Replace
    translationSpace=BCS_WorldSpace
    alphaBoolBlend {blendOption=Linear}
    alphaScaleBiasClamp {inRange {max=0, min=-40}, outRange {max=0, min=1}}
    boneToModify {boneName=thigh_r}
   }
   n11 = "Transform (Modify) Bone - Bone: calf_l"(ComponentPose=@n10, Alpha=1, Translation="0,0,0", Rotation="0,0,0", Scale="1,1,1") {
    rotationMode=BMM_Replace
    rotationSpace=BCS_WorldSpace
    scaleSpace=BCS_BoneSpace
    translationMode=BMM_Replace
    translationSpace=BCS_WorldSpace
    alphaBoolBlend {blendOption=Linear}
    alphaScaleBiasClamp {inRange {max=0, min=-40}, outRange {max=0, min=1}}
    boneToModify {boneName=calf_l}
   }
   n12 = "Transform (Modify) Bone - Bone: calf_r"(ComponentPose=@n11, Alpha=1, Translation="0,0,0", Rotation="0,0,0", Scale="1,1,1") {
    rotationMode=BMM_Replace
    rotationSpace=BCS_WorldSpace
    scaleSpace=BCS_BoneSpace
    translationMode=BMM_Replace
    translationSpace=BCS_WorldSpace
    alphaBoolBlend {blendOption=Linear}
    alphaScaleBiasClamp {inRange {max=0, min=-40}, outRange {max=0, min=1}}
    boneToModify {boneName=calf_r}
   }
   Save cached pose 'CR_Cache_Input'(Pose=Component To Local(ComponentPose=@n12)) {cacheName=CR_Cache_Input}
  ]}
 EventGraph {nodes [
   Event BlueprintUpdateAnimation()
   > n1 = Sequence!(execute) -> (then_0, then_1, then_2:Exec, then_3:Exec)
   > n2 = Gravity Override State Function DEPRECATED(CharacterPartAnimInstance=Self-Reference(), GameTimeInSeconds=GetGameTimeInSeconds(), GravityOverrideParams=Get Gravity Override Params())
   > Set Gravity Override(Gravity Override=vector * vector(A=@n2, B:Vector="1,1,1.5") -> (ReturnValue=~))
   > n3 = Branch(Condition=Get bIsInFrontEnd())
   > Set SimSpaceSettings(SimSpaceSettings:Sim Space Settings Structure=Get SimSpaceFE() -> (SimSpaceFE:Sim Space Settings Structure)) -> (Output_Get:Sim Space Settings Structure)
   Set SimSpaceSettings(execute=@n3.else, SimSpaceSettings=Get SimSpacePIE() -> (SimSpacePIE:Sim Space Settings Structure))
   n4 = PlayerMovement(execute=@n1.then_1, AnimInstance=Self-Reference(), XDotProductCoefficient=100, YDotProductCoefficient=100, ZDotProductCoefficient=100)
   n5 = BreakVector(InVec=MakeVector(X=@n4, Y=@n4.YMovement, Z=@n4.ZMovement))
   > n6 = Set Left(Left:bool="float < float"(A=@n5, B=-50)) -> (Output_Get:bool)
   > n7 = Set Right(Right:bool="float > float"(A=@n5, B=50)) -> (Output_Get:bool)
   > n8 = Branch(Condition=AND Boolean(A=@n7, B=Get bIsCrouching()))
   > Set IsRight(IsRight:double=1) -> (Output_Get:double)
   Set IsRight(execute=@n8.else, IsRight=0)
   n9 = NOT Boolean(A=Get bIsJumping())
   > n10 = Branch(Condition=OR Boolean!(A=@n6, B=Get bIsCrouching(), C:bool=AND Boolean(A=@n9, B=false)) -> (ReturnValue))
   > Set IsLeft(IsLeft:double=1) -> (Output_Get:double)
   Set IsLeft(execute=@n10.else, IsLeft=0)
   n11 = Branch(execute=@n1.then_3, Condition=Get bIsInFrontEnd())
   > Set RbanFEAlpha(RbanFEAlpha:double=1) -> (Output_Get:double)
   > Set RbanAlpha(RbanAlpha:double=0) -> (Output_Get:double)
   Set RbanFEAlpha(execute=@n11.else, RbanFEAlpha=0)
   > Set RbanAlpha(RbanAlpha=1)
   n12 = GetPartSkeletalMeshComponent()
   Is Valid(exec=@n1.then_2, InputObject=@n12)
   n13 = GetAttachParent!(self:Scene Component Object Reference=@n12) -> (ReturnValue:Scene Component Object Reference)
   > Is Valid(InputObject=@n13)
   > Set Body_Mesh(Body_Mesh:Scene Component Object Reference=@n13) -> (Output_Get:Scene Component Object Reference)
   > n14 = Sequence!(execute) -> (then_0, then_1, then_2, then_3, then_4:Exec, then_5:Exec, then_6:Exec)
   n15 = Get Body_Mesh() -> (Body_Mesh:Scene Component Object Reference)
   n16 = BreakTransform(InTransform=GetSocketTransform(self=@n15, InSocketName=root, TransformSpace=RTS_World))
   Set root_location(execute=@n14.then_6, root_location:Vector=@n16) -> (Output_Get:Vector)
   > Set root_space_rotation(root_space_rotation:Rotator=@n16.Rotation) -> (Output_Get:Rotator)
   n17 = BreakTransform(InTransform=GetSocketTransform(self=@n15, InSocketName=pelvis, TransformSpace=RTS_World))
   Set pelvis_location(execute=@n14.then_0, pelvis_location:Vector=@n17) -> (Output_Get:Vector)
   > Set pelvis_rotation(pelvis_rotation:Rotator=@n17.Rotation) -> (Output_Get:Rotator)
   n18 = BreakTransform(InTransform=GetSocketTransform(self=@n15, InSocketName=thigh_l, TransformSpace=RTS_World))
   Set thigh_left_location(execute=@n14.then_1, thigh_left_location:Vector=@n18) -> (Output_Get:Vector)
   > Set thigh_left_rotation(thigh_left_rotation:Rotator=@n18.Rotation) -> (Output_Get:Rotator)
   n19 = BreakTransform(InTransform=GetSocketTransform(self=@n15, InSocketName=thigh_r, TransformSpace=RTS_World))
   Set thigh_right_location(execute=@n14.then_2, thigh_right_location:Vector=@n19) -> (Output_Get:Vector)
   > Set thigh_right_rotation(thigh_right_rotation:Rotator=@n19.Rotation) -> (Output_Get:Rotator)
   n20 = BreakTransform(InTransform=GetSocketTransform(self=@n15, InSocketName=calf_r, TransformSpace=RTS_World))
   Set calf_right_location(execute=@n14.then_4, calf_right_location:Vector=@n20) -> (Output_Get:Vector)
   > Set calf_right_rotation(calf_right_rotation:Rotator=@n20.Rotation) -> (Output_Get:Rotator)
   n21 = BreakTransform(InTransform=GetSocketTransform(self=@n15, InSocketName=calf_l, TransformSpace=RTS_World))
   Set calf_left_location(execute=@n14.then_3, calf_left_location:Vector=@n21) -> (Output_Get:Vector)
   > Set calf_left_rotation(calf_left_rotation:Rotator=@n21.Rotation) -> (Output_Get:Rotator)
   n22 = BreakTransform(InTransform=GetSocketTransform(self=@n15, InSocketName=root, TransformSpace=RTS_World))
   Set body_root_location(execute=@n14.then_5, body_root_location:Vector=@n22) -> (Output_Get:Vector)
   > Set body_root_rotation(body_root_rotation:Rotator=@n22.Rotation) -> (Output_Get:Rotator)
   Event OnBeginCrouching()
   > Set IsCrochingFloat(IsCrochingFloat:double=1) -> (Output_Get:double)
   Event OnCharacterCustomizationLoaded()
   > n23 = Branch(Condition=Get bIsInFrontEnd())
   > Set Pickaxe_Rot(Pickaxe_Rot:double=1) -> (Output_Get:double)
   Set Pickaxe_Rot(execute=@n23.else, Pickaxe_Rot=0)
   Event OnEndPickaxeSwing() {is=title K2Node_Event}
   > Set IsPickaxeSwing(IsPickaxeSwing:double=0.5) -> (Output_Get:double)
   Event OnBeginPickaxeSwing() {is=title K2Node_Event}
   > Set IsPickaxeSwing(IsPickaxeSwing=1)
   Event OnEndCrouching()
   > Set IsCrochingFloat(IsCrochingFloat=0)
   TryGetPawnOwner()
  ]}
]
variables [
 Body_Mesh {type=SceneComponent, default=null}
 body_root_location {type=Vector}
 body_root_rotation {type=Rotator}
 calf_left_location {type=Vector}
 calf_left_rotation {type=Rotator}
 calf_right_location {type=Vector}
 calf_right_rotation {type=Rotator}
 Gravity Override {type=Vector}
 Gravity Override Params {type=GravityOverrideParamsStruct, default {front_End {gravityOverride=<0 0 -1200>}}}
 IsCrochingFloat {type=number}
 IsLeft {type=number}
 IsPickaxeSwing {type=number}
 IsRight {type=number}
 Left {type=boolean}
 pelvis_location {type=Vector}
 pelvis_rotation {type=Rotator}
 Pickaxe_Rot {type=number}
 RbanAlpha {type=number}
 RbanFEAlpha {type=number}
 Right {type=boolean}
 root_location {type=Vector}
 root_space_rotation {type=Rotator}
 SimSpaceFE {
  type=SimSpaceSettings
  default {
   dampingAlpha=0
   maxAngularAcceleration=45
   maxAngularVelocity=8
   maxLinearAcceleration=3500
   maxLinearVelocity=700
   velocityScaleZ=0.3
   worldAlpha=0.4
   externalLinearDragV=<0 0 0>
  }
 }
 SimSpacePIE {
  type=SimSpaceSettings
  default {
   dampingAlpha=0
   maxAngularAcceleration=45
   maxAngularVelocity=8
   maxLinearAcceleration=3500
   maxLinearVelocity=700
   velocityScaleZ=0.5
   worldAlpha=1
   externalLinearDragV=<0.5 0.5 0.35>
  }
 }
 SimSpaceSettings {
  type=SimSpaceSettings
  default {
   dampingAlpha=1
   maxAngularAcceleration=10000
   maxAngularVelocity=10000
   maxLinearAcceleration=10000
   maxLinearVelocity=10000
   velocityScaleZ=1
   worldAlpha=0
   externalLinearDragV=<0 0 0>
  }
 }
 thigh_left_location {type=Vector}
 thigh_left_rotation {type=Rotator}
 thigh_right_location {type=Vector}
 thigh_right_rotation {type=Rotator}
]
